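; Lemon bucket support prop Blueprint that builds on an adjustable light parent, judging by its AdjustableLight variables.
; Its construction script runs the parent's, then adds the Offset vector to the relative location of the LightFixture.
; The same offset shifts either the ISM_DownRod or the DownRod component, selected by the UseGeoForDownrod flag.
; BeginPlay, Tick and overlap events are empty.
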(event EventBeginPlay)

(event Collision|EventActorBeginOverlap
  (params (OtherActor :type Actor%20Object%20Reference)))

(event EventTick
  (params (DeltaSeconds :type Float%20%28single-precision%29)))

(fn ConstructionScript
  (|CallParentFunction)
  (Transformation|SetRelativeLocation (Variables|AdjustableLight|GetLightFixture) (+ (Variables|/Script/Engine.SceneComponent|GetRelativeLocation (Variables|AdjustableLight|GetLightFixture)) (Variables|AdjustableLight|GetOffset)) false false)
  (if (Variables|AdjustableLight|GetUseGeoForDownrod)
    (Transformation|SetRelativeLocation (Variables|AdjustableLight|GetISM_DownRod) (+ (Variables|/Script/Engine.SceneComponent|GetRelativeLocation (Variables|AdjustableLight|GetISM_DownRod)) (Variables|AdjustableLight|GetOffset)) false false)
    (else
      (Transformation|SetRelativeLocation (Variables|AdjustableLight|GetDownRod) (+ (Variables|/Script/Engine.SceneComponent|GetRelativeLocation (Variables|AdjustableLight|GetDownRod)) (Variables|AdjustableLight|GetOffset)) false false))))
